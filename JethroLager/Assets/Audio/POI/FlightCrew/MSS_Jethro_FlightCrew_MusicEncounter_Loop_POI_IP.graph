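# MetaSound Source preset that supplies the looping music for the Jethro FlightCrew POI music encounter.
# It inherits its graph from MSS_MusicEncounter_VolumeLoop and sets the loop wave, tempo, meter, bar length, master volume and pause fade-out time.
# A pause input is exposed alongside the fade time, and the stereo output is assigned to the Fort_Music sound class.
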MetaSoundSource /JethroLager/Assets/Audio/POI/FlightCrew/MSS_Jethro_FlightCrew_MusicEncounter_Loop_POI_IP
preset_parent=/MusicEncounter/MetaSounds/MSS_MusicEncounter_VolumeLoop
inputs [
 MasterVolume {type=Float}
 Music.BeatsPerBar {type=Int32}
 Music.LoopLengthBars {type=Int32}
 Music.Tempo {type=Float}
 Music.WaveAsset {type=WaveAsset}
 MusicEncounter.bIsPaused {type=Bool}
 MusicEncounter.PauseFadeOutTime {type=Time}
 UE.Source.OnPlay {type=Trigger}
]
interfaces [UE.OutputFormat.Stereo:1.0, UE.Source.OneShot:1.0, UE.Source:1.1]
outputs [
 "UE.OutputFormat.Stereo.Audio:0" {type=Audio}
 "UE.OutputFormat.Stereo.Audio:1" {type=Audio}
 UE.Source.OneShot.OnFinished {type=Trigger}
]
pages [Default {input_defaults {
   MasterVolume=0.35
   Music.BeatsPerBar=4
   Music.LoopLengthBars=14
   Music.Tempo=104
   Music.WaveAsset=&/MX_S35_Jethro_FlightCrew_POI_Loop
   MusicEncounter.PauseFadeOutTime=2
   MusicEncounter.bIsPaused=null
   UE.Source.OnPlay=null
  }}]
properties {
 assetImportData=&self:AssetImportData
 numChannels=2
 outputFormat=Stereo
 sampleRate=48000
 soundClassObject=/Game/Sounds/FortSoundClassesAndModes/Fort_Music
}
